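# Physics asset for the CoolSlice Golf two-handed pickaxe cosmetic, apparently providing secondary motion for its paper tag.
# The handle body pairs a long thin capsule along the shaft with a shorter, fatter capsule overlapping it.
# The dyn_paper_tag body is a single capsule joined to the handle by one constraint, with wide swing limits and a very tight twist limit.
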PhysicsAsset /BRCosmetics/Weapons/FORT_Melee/Pickaxe_CoolSlice_Golf/Meshes/CoolSlice_Golf_Axe_Physics
bodies {
 Melee_TwoHanded_Handle [
  {
   shapeType=Capsule
   length=69.8441
   radius=2.5228
   shapeName=Melee_TwoHanded_Handle_capsule
   center=<0.604 -0.711 27.2964>
   rotation=<rot -1.3533 -90 0.0037>
  }
  {shapeType=Capsule, length=15, radius=6, shapeName=Melee_TwoHanded_Handle_capsule, center=<0.8554 0 30.5158>}
 ]
 dyn_paper_tag [{shapeType=Capsule, length=17.5, radius=4.5, shapeName=dyn_paper_tag_capsule, center=<11.645 0 2.947>, rotation=<rot 90 0 0>}]
}
constraints [{bone1Name=dyn_paper_tag, bone2Name=Melee_TwoHanded_Handle, swing1LimitDegrees=120, swing2LimitDegrees=70, twistLimitDegrees=1}]
